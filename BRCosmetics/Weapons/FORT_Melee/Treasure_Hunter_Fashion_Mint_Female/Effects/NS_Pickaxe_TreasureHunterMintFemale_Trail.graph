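# Niagara ribbon trail effect for the female Treasure Hunter Fashion Mint pickaxe cosmetic.
# A single emitter, ThinTrails002, spawns short-lived anim-trail particles that follow bones on the User.SkeletalMesh input.
# Curves shift color from bright mint toward darker teal, swell then narrow the ribbon width, and ramp a dynamic material parameter.
# Ribbon U is overridden from emitter age, with a small camera offset; curl noise is present but disabled.
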NiagaraSystem /BRCosmetics/Weapons/FORT_Melee/Treasure_Hunter_Fashion_Mint_Female/Effects/NS_Pickaxe_TreasureHunterMintFemale_Trail
dependencies {
 usedDataInterfaces [&NiagaraDataInterfaceColorCurve, &NiagaraDataInterfaceCurve]
 usedDynamicInputs []
 usedModules [&SystemState]
}
emitters {ThinTrails002 {
  emitterUpdateScript [
   &EmitterState(Life Cycle Mode=::NewEnumerator1, Loop Duration=1, UseLoopDelay=false)
   &FN_SpawnAnimTrail()
  ]
  particleSpawnScript [
   &FN_InitializeAnimTrailBoneIndecies()
   &FN_InitializeAnimTrail(AnimTrailWidth=25, Lifetime=0.3)
   /SetVariables(Particles.FrontRibbonMask:NiagaraFloat=&OneMinusFloat(Float=/Niagara/DynamicInputs/Integer/V2/Round_Float(Float:NiagaraFloat=$System.NormalizedLoopAge))) {isSetParameters=true}
  ]
  particleUpdateScript [
   &FN_UpdateAnimTrail(SkeletalMesh=$User.SkeletalMesh)
   &ScaleColor(
    Scale Alpha=&FloatFromCurve(FloatCurve=DataInterface() {propertyValues {Curve {keys [{tangentMode=RCTM_User, time=0, value=1}]}}})
    Scale RGB=&MakeVectorFromLinearColorRGB(Color=&ColorFromCurve(ColorCurve=DataInterface() {propertyValues {
        AlphaCurve {keys [<at 0 1 RCIM_Cubic>]}
        BlueCurve {keys [<at 0.0031 0.8304>,<at 0.7414 0.7>]}
        GreenCurve {keys [<at 0.0031 1.5>,<at 0.7414 0.6318>]}
        RedCurve {keys [<at 0.0031 0.1083>,<at 0.7414 0>]}
       }}))
    ScaleA=true
    ScaleRGB=true
   )
   &ParticleState()
   &ScaleRibbonWidth(Ribbon Width Scale=&FloatFromCurve(FloatCurve=DataInterface() {propertyValues {Curve {keys [
         {arriveTangent=3.0944, interpMode=RCIM_Cubic, leaveTangent=3.0944, tangentMode=RCTM_User, time=0, value=0.5}
         {arriveTangent=-0.156, interpMode=RCIM_Cubic, leaveTangent=-0.156, tangentMode=RCTM_User, time=0.25, value=1}
         {arriveTangent=-0.977, interpMode=RCIM_Cubic, leaveTangent=-0.977, tangentMode=RCTM_User, time=1, value=0.5}
        ]}}}))
   /SetVariables(Particles.RibbonU0Override=&FloatFromCurve(
     FloatCurve=DataInterface() {propertyValues {Curve {keys [{tangentMode=RCTM_User, time=0, value=0}, {tangentMode=RCTM_User, time=1, value=1}]}}}
     CurveIndex=$Emitter.Age
    )) {isSetParameters=true}
   &DynamicMaterialParameters(
    Index 1 Param 4=&FloatFromCurve(FloatCurve=DataInterface() {propertyValues {Curve {keys [
         {interpMode=RCIM_Cubic, tangentMode=RCTM_User, time=0, value=0}
         {arriveTangent=0.65, interpMode=RCIM_Cubic, leaveTangent=0.65, time=0.4, value=0.35}
         {arriveTangent=0.4394, interpMode=RCIM_Cubic, leaveTangent=0.4394, tangentMode=RCTM_User, time=1, value=0.55}
        ]}}})
    Module.Param0WriteEnabled1=false
    Module.Param1WriteEnabled1=false
    Module.Param2WriteEnabled1=false
    Module.Param3WriteEnabled1=true
    Write Parameter Index 0=false
    Write Parameter Index 1=true
   )
   &CameraOffset(Camera Offset Amount=10)
   &CurlNoiseForce(Noise Frequency=15, Noise Strength=100, Pan Noise=false) {enabled=false}
   &SolveForcesAndVelocity(Clamp Velocity=false, Limit Acceleration=false)
  ]
  rendererClasses [&NiagaraRibbonRendererProperties]
  renderers [{rendererClass=&NiagaraRibbonRendererProperties}]
 }}
summary {
 systemName=NS_Pickaxe_TreasureHunterMintFemale_Trail
 emitters [{emitterName=ThinTrails002, rendererClasses [&NiagaraRibbonRendererProperties]}]
 userVariables [
  User.LifetimeCustom {type=NiagaraFloat, defaultValue=0, description=""}
  User.LightColor {type=LinearColor, description="", defaultValue {struct=LinearColor, value=<rgba 1 1 1 1>}}
  User.RibbonIDOffset {type=NiagaraInt32}
  User.SkeletalMesh {
   type=NiagaraDataInterfaceSkeletalMesh
   defaultValue {
    struct=NiagaraExt_VariableValue_DataInterface
    value {dataInterface=&self:NiagaraDataInterfaceSkeletalMesh_0}
   }
  }
 ]
}
