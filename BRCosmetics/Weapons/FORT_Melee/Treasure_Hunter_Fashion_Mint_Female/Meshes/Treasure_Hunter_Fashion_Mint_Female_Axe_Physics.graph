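# Physics asset for the Treasure Hunter Fashion Mint Female two-handed melee axe, modeling a medal that dangles from the handle.
# Capsule bodies form a constraint chain from the handle through three string segments and a ring to the medal, which uses three capsules.
# All joints swing freely, with twist free on the top two and locked on the lower three, so the stored limit angles are inactive.
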PhysicsAsset /BRCosmetics/Weapons/FORT_Melee/Treasure_Hunter_Fashion_Mint_Female/Meshes/Treasure_Hunter_Fashion_Mint_Female_Axe_Physics
bodies {
 Melee_TwoHanded_Handle [
  {shapeType=Capsule, length=37.0826, radius=2.75, center=<0 0 26.8736>}
  {shapeType=Capsule, length=5.25, radius=3, center=<0.0001 2.0845 47.2892>, rotation=<rot 0 0 60.0006>}
 ]
 dyn_medal_01 [
  {shapeType=Capsule, length=9.0292, radius=1, center=<-2.867 1.6553 -7.1831>, rotation=<rot 0 -30.0003 0>}
  {shapeType=Capsule, length=9, radius=1, shapeName=dyn_medal_01_capsule, center=<2.6838 -1.5099 -7.2898>, rotation=<rot 0 -30.0003 0>}
  {shapeType=Capsule, length=5.25, radius=1, shapeName=dyn_medal_01_capsule, center=<0 0.0001 -11.4884>, rotation=<rot 90 0 30.0037>}
 ]
 dyn_ring_01 [{shapeType=Capsule, length=1.1927, radius=1.25, center=<0 0 -1.7499>}]
 dyn_string_01 [{shapeType=Capsule, length=1.005, radius=2, center=<0.7728 0 -3.1233>, rotation=<rot 20.0001 0 0>}]
 dyn_string_02 [{shapeType=Capsule, length=1.786, radius=1.75, center=<0.328 -0.0001 -2.8602>, rotation=<rot 10.0001 0 0>}]
 dyn_string_03 [{shapeType=Capsule, length=0.3204, radius=1.75, center=<0.1338 -0.0001 -1.7588>, rotation=<rot 10.0001 0 0>}]
}
constraints [
 {
  bone1Name=dyn_string_02
  bone2Name=dyn_string_01
  swing1LimitDegrees=45
  swing1Motion=Free
  swing2LimitDegrees=45
  swing2Motion=Free
  twistLimitDegrees=45
  twistMotion=Free
 }
 {
  bone1Name=dyn_string_03
  bone2Name=dyn_string_02
  swing1LimitDegrees=45
  swing1Motion=Free
  swing2LimitDegrees=45
  swing2Motion=Free
  twistLimitDegrees=45
  twistMotion=Locked
 }
 {
  bone1Name=dyn_ring_01
  bone2Name=dyn_string_03
  swing1LimitDegrees=45
  swing1Motion=Free
  swing2LimitDegrees=45
  swing2Motion=Free
  twistLimitDegrees=20
  twistMotion=Locked
 }
 {
  bone1Name=dyn_medal_01
  bone2Name=dyn_ring_01
  swing1LimitDegrees=45
  swing1Motion=Free
  swing2LimitDegrees=45
  swing2Motion=Free
  twistLimitDegrees=20
  twistMotion=Locked
 }
 {
  bone1Name=dyn_string_01
  bone2Name=Melee_TwoHanded_Handle
  swing1LimitDegrees=45
  swing1Motion=Free
  swing2LimitDegrees=45
  swing2Motion=Free
  twistLimitDegrees=45
  twistMotion=Free
 }
]
